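# Physics asset for the sports car vehicle mesh, giving its antenna a simulated body linked to the car body.
# A long, thin capsule runs along the antenna and a small capsule stands in for the body bone.
# One constraint joins antenna to body with twist locked and narrow swing limits set for both swing axes.
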PhysicsAsset /Valet/SportsCar/Mesh/Sports_Car_Veh_Dyn
bodies {
 antenna [{shapeType=Capsule, length=85.4237, radius=0.706, center=<40 0 0>, rotation=<rot 90 0 0>}]
 body [{shapeType=Capsule, length=5, radius=5}]
}
constraints [{bone1Name=antenna, bone2Name=body, swing1LimitDegrees=20, swing2LimitDegrees=20, twistLimitDegrees=45, twistMotion=Locked}]
