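# Physics asset for the damaged right mirror of the sports car vehicle, with capsule bodies for body, frame, right mirror and right exhaust.
# Constraints chain the mirror to the body and the body to the frame, with the exhaust branching separately off the frame.
# The body-to-frame joint rotates freely on all axes, while the mirror has modest angular limits and the exhaust has tight swing limits with locked twist.
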PhysicsAsset /Valet/SportsCar/Mesh/Sports_Car_Veh_Dyn_dmg_mirror_r
bodies {
 body [{shapeType=Capsule, length=11.9688, radius=10.4375, center=<-3.643 0.0005 -0.0002>, rotation=<rot 0 0 89.9997>}]
 exhaust_r [{shapeType=Capsule, length=90.5, radius=10, center=<57.1886 -0.0002 0.0028>, rotation=<rot 90 0 0>}]
 frame [{shapeType=Capsule, length=11.9688, radius=10.4375, center=<-3.643 0.0012 49.8688>, rotation=<rot 0 0 89.9997>}]
 mirror_r [{shapeType=Capsule, length=14.6562, radius=10, center=<0.0003 15.4481 -2.8592>, rotation=<rot 0 0 -90>}]
}
constraints [
 {bone1Name=mirror_r, bone2Name=body, swing1LimitDegrees=10, swing2LimitDegrees=30, twistLimitDegrees=20}
 {
  bone1Name=body
  bone2Name=frame
  swing1LimitDegrees=45
  swing1Motion=Free
  swing2LimitDegrees=45
  swing2Motion=Free
  twistLimitDegrees=45
  twistMotion=Free
 }
 {bone1Name=exhaust_r, bone2Name=frame, swing1LimitDegrees=5, swing2LimitDegrees=10, twistLimitDegrees=45, twistMotion=Locked}
]
