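# Physics asset for the sports car's dynamic damage spoiler mesh, giving the breakable spoiler piece simulated collision and articulation.
# It has two capsule bodies, a shorter, thicker one on the body bone and a long, thin one on spoiler_break.
# A single constraint joins spoiler_break to body, locking the second swing axis so the spoiler's rotation relative to its base is restricted.
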PhysicsAsset /Valet/SportsCar/Mesh/Sports_Car_Veh_Dyn_dmg_spoiler
bodies {
 body [{shapeType=Capsule, length=77.8438, radius=40, center=<-3.643 0.0005 -0.0002>, rotation=<rot 0 0 89.9997>}]
 spoiler_break [{shapeType=Capsule, length=158.9684, radius=9.462, center=<1.1526 -65.7952 7.0366>, rotation=<rot 0 0 -90>}]
}
constraints [{bone1Name=spoiler_break, bone2Name=body, swing1LimitDegrees=10, swing2LimitDegrees=0, swing2Motion=Locked, twistLimitDegrees=20}]
