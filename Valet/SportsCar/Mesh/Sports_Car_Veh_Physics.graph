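# Physics asset for the sports car vehicle mesh in the Valet SportsCar folder.
# It defines a single body on the root bone with one box collision shape, named DashCollision, offset forward and up from the bone.
# There are no constraints, so the car's collision is one rigid box with no articulated parts such as wheels.
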PhysicsAsset /Valet/SportsCar/Mesh/Sports_Car_Veh_Physics
bodies {root [{
   shapeType=Box
   extentX=71.5
   extentY=183.5
   extentZ=20.5
   shapeName=DashCollision
   center=<80.849 0.0001 80.8182>
   rotation=<rot 0 0 -0.0004>
  }]}
constraints []
