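# Physics asset for the Valet sports car skeletal mesh, a variant named for fluid simulation.
# Approximates the car with one box per bone: body, glass, spoiler and four wheels, plus a tiny box on the frame bone.
# Defines no constraints between the bodies, so it provides coarse collision shapes without any joint setup.
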PhysicsAsset /Valet/SportsCar/Mesh/Sports_Car_Veh_Physics_Fluidsim
bodies {
 body [{shapeType=Box, extentX=180.6597, extentY=62.4088, extentZ=410.9205, center=<5.0972 -1.4563 6.8158>, rotation=<rot 0 -91.0116 -86.4457>}]
 frame [{shapeType=Box, extentX=1.01, extentY=1.01, extentZ=1.01, rotation=<rot 0 -90 0>}]
 glass_c [{shapeType=Box, extentX=176.2391, extentY=32.6876, extentZ=148.8782, center=<-140.9257 -0.5982 11.6636>, rotation=<rot 0 -179.7597 -89.8836>}]
 spoiler_break [{shapeType=Box, extentX=50.7263, extentY=25.2463, extentZ=194.6693, center=<-2.8666 -63.1746 2.0664>, rotation=<rot 0 179.1267 -89.6682>}]
 wheel_tilted_bk_l [{shapeType=Box, extentX=55.7266, extentY=58.1672, extentZ=58.4992, center=<0.2103 19.8953 -0.9431>, rotation=<rot 0 -0.007 -85.0184>}]
 wheel_tilted_bk_r [{shapeType=Box, extentX=55.7267, extentY=58.1289, extentZ=58.4603, center=<-0.3551 -19.2614 -0.9997>, rotation=<rot 0 179.6487 -85.0645>}]
 wheel_tilted_fr_l [{shapeType=Box, extentX=35.2486, extentY=50.1706, extentZ=50.2929, center=<-0.0132 8.2954 0.1173>, rotation=<rot -0.8662 -90 0.2208>}]
 wheel_tilted_fr_r [{shapeType=Box, extentX=35.2486, extentY=50.1706, extentZ=50.2929, center=<-0.0132 -7.0209 0.0057>, rotation=<rot 0.8662 -90 0.2208>}]
}
constraints []
