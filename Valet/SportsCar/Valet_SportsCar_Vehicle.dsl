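; Sports car Valet vehicle Blueprint, adding driver feedback and headlight lens flares to its parent.
; Gear changes play force feedback and camera shake, with a brief FOV offset on upshifts, all suppressed during and one second after boost, which adds looping cue VFX.
; Light toggles lazily spawn or switch headlight Niagara flares, counted from existing sockets.
; Construction applies hotfixable pontoon tick-rate scales.
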(event EventBeginPlay
  (|CallParentFunction))

(event EventCurrentGearCheck
  (params (NewGear :type Integer))
  ;[1 1 1] FOV changes and FF on shifts
  (|CallParentFunction NewGear)
  (bind _upshift (CallFunction|HandleShifting NewGear))
  ;; HandleFOVDuringShift
  (if _upshift
    (Variables|Camera|SetCameraFOVOffset -5.0)
    (Utilities|FlowControl|Delay 0.2)
    (Variables|Camera|SetCameraFOVOffset 0.0)))

(event EventCurrentGearCheckCosmetic
  (params (NewGear :type Integer) (IgnoreExhaustPops :type Boolean))
  ;[1 1 1] FOV changes and FF on shifts
  (|CallParentFunction NewGear IgnoreExhaustPops)
  (bind _upshift (CallFunction|HandleShifting 0))
  ;; HandleFOVDuringShift_3
  (if _upshift
    (Variables|Camera|SetCameraFOVOffset -5.0)
    (Utilities|FlowControl|Delay 0.2)
    (Variables|Camera|SetCameraFOVOffset 0.0)))

(event EventOnBoostStarted
  (bind _driverpawn (Variables|ValetDefaults|GetDriverPawn))
  (bind _returnvalue (Fortnite|GetFortPlayerControllerfromActor _driverpawn))
  ;[1 1 1] Boost Event > VFX > FF > State Set
  (|CallParentFunction)
  (Utilities|IsValid _returnvalue
    (:Is%20Valid
      (Variables|Default|SetbHasBoosted true)
      (Game|Feedback|ClientPlayForceFeedback _returnvalue (Variables|Default|GetFF_CatchUpBoost) "None" false false false)
      (Projectile|AddGameplayCue%28Looping%29Local _returnvalue (Variables|Default|GetGC_CatchUpBoost) (Ability|GameplayCue|MakeGameplayCueParameters 0.0)))))

(event EventOnBoostFinished
  (bind _driverpawn (Variables|ValetDefaults|GetDriverPawn))
  (bind _returnvalue (Fortnite|GetFortPlayerControllerfromActor _driverpawn))
  ;[1 1 1] Boost End Event > Remove VFX > Reset State || Delay is to make sure FOV reset back to normal so shifting doesnt affect it
  (|CallParentFunction)
  (Utilities|IsValid _returnvalue
    (:Is%20Valid
      (Projectile|RemoveGameplayCue%28Looping%29Local _returnvalue (Variables|Default|GetGC_CatchUpBoost) (Ability|GameplayCue|MakeGameplayCueParameters 0.0))
      (Utilities|FlowControl|Delay 1.0)
      (Variables|Default|SetbHasBoosted false))))

(event EventHandleOnExplode
  (params (LastDamageInstigator :type Controller%20Object%20Reference) (Vehicle :type Actor%20Object%20Reference))
  ;[1 1 1] Event ----> Handle Lens Flare on explosion and headlight change
  (|CallParentFunction LastDamageInstigator Vehicle)
  (bind _spawned_headlights_niagara (Variables|Default|GetSpawned%20Headlights%20Niagara)
    (:then
      (Components|Activation|Deactivate _spawned_headlights_niagara))
    (:else)))

(event OnLightEnabledChanged(ValetLightSpawnerComponent)
  (params (enabled :type Boolean))
  (bind (_headlightcount _headlightcolorarray) (Utilities|Struct|BreakHeadlightEffectInfo (Vehicle|Cosmetics|GetHeadlightEffectInfo)))
  ;[1 1 1] Event ----> Handle Lens Flare on explosion and headlight change
  (if enabled
    (bind _spawned_headlights_niagara (Variables|Default|GetSpawned%20Headlights%20Niagara)
      (:then
        (Components|Activation|Activate _spawned_headlights_niagara false)
        (Rendering|SetVisibility _spawned_headlights_niagara true false))
      (:else
        (bind _returnvalue (Niagara|SpawnSystemAttached (Variables|FX|GetHeadlights%20Niagara%20System) (Variables|Vehicle|GetSkeletalMesh) "root"))
        (bind _output_get (Variables|Default|SetSpawned%20Headlights%20Niagara _returnvalue))
        (Default|SetNumberOfFlarestoHeadlights _output_get)
        (Niagara|NiagaraSetColorArray _output_get "FlareColorArray" _headlightcolorarray)))
    (else
      (bind _spawned_headlights_niagara_2 (Variables|Default|GetSpawned%20Headlights%20Niagara)
        (:then
          (Components|Activation|Deactivate _spawned_headlights_niagara_2)
          (Rendering|SetVisibility _spawned_headlights_niagara_2 false false))
        (:else)))))

(event Vehicle|Cosmetics|EventOnVehicleCosmeticsFinished
  (bind _spawned_headlights_niagara (Variables|Default|GetSpawned%20Headlights%20Niagara))
  ;[1 1 1] Update le number of headlights when changing car.
  (|CallParentFunction)
  (Utilities|IsValid _spawned_headlights_niagara
    (:Is%20Valid
      (Default|SetNumberOfFlarestoHeadlights _spawned_headlights_niagara))))

(fn HandleShifting
  (params (NewGear :type Integer))
  (bind _driverpawn (Variables|ValetDefaults|GetDriverPawn))
  (bind _returnvalue (Fortnite|GetFortPlayerControllerfromActor _driverpawn))
  (bind _driverpawn_2 (Variables|ValetDefaults|GetDriverPawn))
  (bind _returnvalue_2 (Fortnite|GetFortPlayerControllerfromActor _driverpawn_2))
  (if (Variables|Default|GetbHasBoosted)
    (return false)
    (elif (> NewGear (Variables|Default|GetCurrentGearAudio))
      (Game|Feedback|ClientPlayForceFeedback _returnvalue_2 (Variables|Default|GetFF_Upshift) "None" false false false)
      (Game|Feedback|ClientStartCameraShake _returnvalue_2 (Variables|Default|GetUpShiftCameraShake) 1.0 "CameraLocal" (Transformation|GetActorRotation _driverpawn_2))
      (Variables|Default|SetCurrentGearAudio NewGear)
      (return true)
      (else
        (Game|Feedback|ClientPlayForceFeedback _returnvalue (Variables|Default|GetFF_Downshift) "None" false false false)
        (Game|Feedback|ClientStartCameraShake _returnvalue (Variables|Default|GetDownShiftCameraShake) 1.0 "CameraLocal" (Transformation|GetActorRotation))
        (Variables|Default|SetCurrentGearAudio NewGear)
        (return false)))))

(fn SetNumberOfFlaresToHeadlights
  (params (Niagara%20system :type Niagara%20Particle%20System%20Component%20Object%20Reference))
  (bind _existingsocketslist (LocalVariables|GetExistingSocketsList))
  (bind _returnvalue (Utilities|Array|Length _existingsocketslist))
  (bind _asfort_vehicle_skel_mesh_component (Utilities|Casting|CastToFortVehicleSkelMeshComponent (Variables|Vehicle|GetSkeletalMesh)))
  (for _array_element (LocalVariables|GetFilteredSocketsList)
    (if (Transformation|DoesSocketExist _asfort_vehicle_skel_mesh_component _array_element)
      (Utilities|Array|Add (LocalVariables|GetExistingSocketsList) _array_element)
      (Niagara|SetNiagaraVariable(Int32) Niagara%20system "FlaresCount" _returnvalue)
      (Development|PrintString (Utilities|String|Append "Car lensflares Headligth: " (Utilities|String|ToString(Integer) _returnvalue)) true true "(R=1.000000,G=0.000000,B=0.145770,A=1.000000)" 4.0))))

(fn ConstructionScript
  (|CallParentFunction)
  (Variables|PontoonScaling|SetScalePontoonTickRateFront (ScalableFloat|GetValueatLevel (Variables|Default|GetHotfixScalePontoonTickRateFront_0)))
  (Variables|PontoonScaling|SetScalePontoonTickRateRear (ScalableFloat|GetValueatLevel (Variables|Default|GetHotfixScalePontoonTickRateRear_0))))
