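; Child Blueprint of a valet sports car vehicle, named as the white camo variant.
; It declares the EventPickRandomMaterial event with an empty body, so the event adds no material logic of its own.
; The construction script only calls the parent version, and no material assignment appears in this graph.
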(event EventPickRandomMaterial)

(fn ConstructionScript
  (|CallParentFunction))
